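# Physics asset for the Hurtle gadget prop, a four-wheeled object, giving its skeletal mesh simplified collision and jointed bodies.
# One large capsule covers the main hydraulics bone, and a smaller rotated capsule covers each front and rear wheel bone.
# Each wheel is constrained directly to the hydraulics body with identical swing and twist limits, forming a star layout with no wheel-to-wheel constraints.
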PhysicsAsset /BRCosmetics/Animation/Props/Gadgets_Hurtle/Meshes/Hurtle_Gadget_Physics
bodies {
 JNT_Hydraulics_Main [{shapeType=Capsule, length=89.9368, radius=43.8208, center=<-0.8363 -0.1161 3.4126>, rotation=<rot 0 178.3162 -89.8102>}]
 JNT_LF_FrontWheel [{shapeType=Capsule, length=16.2893, radius=16.2787, center=<0.014 0 0.0027>, rotation=<rot 0 -90 -0.1134>}]
 JNT_LF_RearWheel [{shapeType=Capsule, length=16.2785, radius=16.2264, center=<0.0697 -0.0011 -0.0138>, rotation=<rot 4.5948 -90 0.5667>}]
 JNT_RT_FrontWheel [{shapeType=Capsule, length=16.2893, radius=16.2787, center=<-0.014 -0.3058 0.0027>, rotation=<rot 0 -90 0.1134>}]
 JNT_RT_RearWheel [{shapeType=Capsule, length=16.2893, radius=16.2787, center=<-0.014 -0.3058 0.0027>, rotation=<rot 0.0001 -90 0.1134>}]
}
constraints [
 {bone1Name=JNT_LF_FrontWheel, bone2Name=JNT_Hydraulics_Main, swing1LimitDegrees=45, swing2LimitDegrees=45, twistLimitDegrees=45}
 {bone1Name=JNT_LF_RearWheel, bone2Name=JNT_Hydraulics_Main, swing1LimitDegrees=45, swing2LimitDegrees=45, twistLimitDegrees=45}
 {bone1Name=JNT_RT_FrontWheel, bone2Name=JNT_Hydraulics_Main, swing1LimitDegrees=45, swing2LimitDegrees=45, twistLimitDegrees=45}
 {bone1Name=JNT_RT_RearWheel, bone2Name=JNT_Hydraulics_Main, swing1LimitDegrees=45, swing2LimitDegrees=45, twistLimitDegrees=45}
]
